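# Physics asset for the CraneLaugh backpack cosmetic, setting up two dangling chains with simple capsule bodies.
# Chain A has four links and chain B two, each linked in series to its own master anchor body.
# Constraints allow narrow swing and very little twist, with a looser swing where chain B meets its anchor.
# Unconstrained capsules on the pack, two spine bones and left upper arm approximate the shapes around the chains.
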PhysicsAsset /BRCosmetics/Accessories/FORT_Backpacks/Backpack_F_MED_CraneLaugh/Meshes/F_MED_CraneLaugh_Pack_Physics
bodies {
 base_spine_01 [{shapeType=Capsule, length=14, radius=11.5, shapeName=spine_01_capsule, center=<1.4991 0 0>}]
 base_spine_04 [{shapeType=Capsule, length=17, radius=10, shapeName=spine_04_capsule, center=<-0.0671 0 0>}]
 base_upperarm_l [{shapeType=Capsule, length=17, radius=9.5, shapeName=upperarm_l_capsule, center=<7.3398 0.0001 0>, rotation=<rot 90 180 179.9995>}]
 dyn_chain_A_01 [{shapeType=Capsule, length=4.1, radius=0.8, shapeName=dyn_strap_01_L_capsule, center=<2.8162 0 0>, rotation=<rot 90 -179.9899 -179.9903>}]
 dyn_chain_A_02 [{shapeType=Capsule, length=4.1, radius=0.8, shapeName=dyn_strap_01_L_capsule, center=<2.8307 0 0>, rotation=<rot 90 -0.0153 -0.0157>}]
 dyn_chain_A_03 [{shapeType=Capsule, length=4.1, radius=0.8, shapeName=dyn_strap_01_L_capsule, center=<2.9242 0 0>, rotation=<rot 90 -0.0293 -0.0297>}]
 dyn_chain_A_04 [{shapeType=Capsule, length=4.1, radius=0.8, shapeName=dyn_strap_01_L_capsule, center=<3.0556 0 0>, rotation=<rot 90 -0.0153 -0.0157>}]
 dyn_chain_B_01 [{shapeType=Capsule, length=4.1, radius=0.8, shapeName=dyn_strap_01_L_capsule, center=<3.1779 0 0>, rotation=<rot 90 -179.9804 -179.9808>}]
 dyn_chain_B_02 [{shapeType=Capsule, length=4.1, radius=0.8, shapeName=dyn_strap_01_L_capsule, center=<3.2841 0 0>, rotation=<rot 90 -179.9804 -179.9808>}]
 mstr_chain_A [{shapeType=Capsule, length=0, radius=0.5, shapeName=strap_MSTR_L_capsule, rotation=<rot 90 -90.0001 -90.0005>}]
 mstr_chain_B [{shapeType=Capsule, length=0, radius=0.5, shapeName=strap_MSTR_L_capsule, rotation=<rot 90 -90.0001 -90.0005>}]
 pack [{
   shapeType=Capsule
   length=20.4823
   radius=35.4513
   shapeName=strap_MSTR_R_capsule
   center=<-0.0002 33.2064 -6.3946>
   rotation=<rot 0 -179.9996 -180>
  }]
}
constraints [
 {bone1Name=dyn_chain_A_04, bone2Name=dyn_chain_A_03, swing1LimitDegrees=15, swing2LimitDegrees=15, twistLimitDegrees=2}
 {bone1Name=dyn_chain_A_03, bone2Name=dyn_chain_A_02, swing1LimitDegrees=15, swing2LimitDegrees=15, twistLimitDegrees=2}
 {bone1Name=dyn_chain_A_02, bone2Name=dyn_chain_A_01, swing1LimitDegrees=15, swing2LimitDegrees=15, twistLimitDegrees=2}
 {bone1Name=mstr_chain_A, bone2Name=dyn_chain_A_01, swing1LimitDegrees=15, swing2LimitDegrees=15, twistLimitDegrees=2}
 {bone1Name=dyn_chain_B_02, bone2Name=dyn_chain_B_01, swing1LimitDegrees=15, swing2LimitDegrees=15, twistLimitDegrees=2}
 {bone1Name=mstr_chain_B, bone2Name=dyn_chain_B_01, swing1LimitDegrees=30, swing2LimitDegrees=30, twistLimitDegrees=2}
]
